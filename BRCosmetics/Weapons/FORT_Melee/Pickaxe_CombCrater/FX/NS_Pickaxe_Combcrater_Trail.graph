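# Niagara weapon trail for the CombCrater pickaxe, drawing short-lived grey ribbons.
# A single MainTrail emitter uses FN_ anim-trail modules to spawn, initialize and update particles, and feeds two ribbon renderers.
# Over particle life, curves scale width and fade alpha.
# The main color is partly tinted by System.TODColor, while a separate DarkTrailColor attribute multiplies near-black by that color with its own alpha fade.
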NiagaraSystem /BRCosmetics/Weapons/FORT_Melee/Pickaxe_CombCrater/FX/NS_Pickaxe_Combcrater_Trail
dependencies {
 usedDataInterfaces [&NiagaraDataInterfaceCurve, &NiagaraDataInterfaceSocketReader, &NiagaraDataInterfaceColorCurve]
 usedDynamicInputs [&ColorFromCurve]
 usedModules [
  &self:SystemSpawnScript.SystemScriptSource.SystemScriptGraph.NiagaraNodeAssignment_1.SetVariables_#1
  &SystemState
  &self:SystemSpawnScript.SystemScriptSource.SystemScriptGraph.NiagaraNodeAssignment_0.SetVariables_#2
 ]
}
emitters {MainTrail {
  emitterUpdateScript [
   &EmitterState(Life Cycle Mode=::NewEnumerator1, Loop Behavior=::NewEnumerator1, Loop Duration=0.25, UseLoopDelay=true)
   &FN_SpawnAnimTrail()
  ]
  particleSpawnScript [
   &FN_InitializeAnimTrailBoneIndecies() {module=FN_InitializeAnimTrailBoneIndecies001}
   &FN_InitializeAnimTrail(
    Color=&MultiplyLinearColor(
     Linear Color=<rgba 0.25 0.25 0.2449 1>
     Scale Factor=&MakeFloat4FromVectorAndFloat("Vector (RGB)"=<1 1 1>, "Float (Alpha)"=1)
    )
    Lifetime=0.175
   )
  ]
  particleUpdateScript [
   &FN_UpdateAnimTrail()
   &ScaleRibbonWidth(Ribbon Width Scale=&FloatFromCurve(FloatCurve=DataInterface() {propertyValues {Curve {keys [{arriveTangent=-0.8703, time=0.2, value=1}, <at 1 0.8>]}}}))
   &ScaleColor(
    Scale Alpha=&FloatFromCurve(FloatCurve=DataInterface() {propertyValues {Curve {keys [{arriveTangent=19.9617, interpMode=RCIM_Cubic, tangentMode=RCTM_User, time=0, value=1}, <at 0.45 0>, <at 1 0>]}}})
    Scale RGB=&Lerp_Vector(A=<1 1 1>, B=&MakeVectorFromLinearColorRGB(Color=$System.TODColor), Alpha=0.7)
    ScaleA=true
    ScaleRGB=true
   ) {module=ScaleColor003}
   &SpriteFacingAndAlignment(Sprite Facing=$Particles.RibbonFacing, WriteAlignment=false, WriteFacing=true)
   /SetVariables(Particles.DarkTrailColor:LinearColor=&MakeLinearColorFromVectorAndFloat(
     "Vector (RGB)"=&Multiply_Vector(
      A=&MakeVectorFromLinearColorRGB(Color=<rgba 0.075 0.075 0.075 1>)
      B=&MakeVectorFromLinearColorRGB(Color=$System.TODColor)
     )
     "Float (Alpha)"=&FloatFromCurve(
      FloatCurve=DataInterface() {propertyValues {Curve {keys [{interpMode=RCIM_Cubic, tangentMode=RCTM_User, time=0, value=1}, {interpMode=RCIM_Cubic, tangentMode=RCTM_User, time=0.75, value=0}]}}}
      Scale Curve=2
     )
    )) {isSetParameters=true}
  ]
  rendererClasses [&NiagaraRibbonRendererProperties]
  renderers [
   {rendererClass=&NiagaraRibbonRendererProperties}
   {rendererClass=&NiagaraRibbonRendererProperties, rendererIndex=1}
  ]
 }}
summary {
 systemName=NS_Pickaxe_Combcrater_Trail
 emitters [{emitterName=MainTrail, rendererClasses [&NiagaraRibbonRendererProperties]}]
 userVariables [
  User.bIsFrontEnd {type=NiagaraBool}
  User.bIsFrontEndPreview {type=NiagaraBool}
  User.RibbonIDOffset {type=NiagaraInt32}
  User.SwingSectionIndex {type=NiagaraInt32}
 ]
}
